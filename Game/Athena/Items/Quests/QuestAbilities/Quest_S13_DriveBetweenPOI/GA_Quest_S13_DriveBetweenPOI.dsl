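; Gameplay ability for a Season 13 quest to drive a required vehicle from a start POI to a destination POI within a time limit.
; On server activation it checks vehicle and location tags, then starts a timer, shows UI and waits for the destination tag.
; Reaching it sends the quest stat event and plays victory FX.
; Exiting the vehicle starts a delayed cleanup, cancelled by re-entering a vehicle with matching tags.
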(event Ability|EventActivateAbilityFromEvent
  (params (EventData :type Gameplay%20Event%20Data%20Structure%20%28by%20ref%29))
  ;[1 1 1] Activate
  (|CallParentFunction EventData)
  (if (Networking|IsServer)
    ;[1 1 1] Ensure player is in the correct vehicle and start location
    (bind _vehicle_occupied (Utilities|IsinVehicle (Variables|GAVars|Data|GetOwningPlayerPawn))
      (:true
        (Variables|Default|SetQuestVehicle _vehicle_occupied)
        (if (and (GameplayTags|HasAllTags (Class|FortVehicleInterface|GetGameplayTags _vehicle_occupied) (Variables|ReadOnly|GetRequiredVehicleTags)) (GameplayTags|HasAllTags (GameplayTags|GetOwnedGameplayTags (Variables|GAVars|Data|GetOwningPlayerPawn) :!class /Script/GameplayTags.GameplayTagAssetInterface) (Variables|ReadOnly|GetStartLocationTags)))
          ;[1 1 1] Listen for destination tag
          (bind _asynctaskproxy (Ability|Tasks|WaitGameplayTagAdd (Variables|ReadOnly|GetDestinationLocationTag) 0 true)
            (:then
              (Variables|Default|SetTagListenTask _asynctaskproxy)
              ;[1 1 1] Set timer
              (bind _returnvalue (Utilities|Time|SetTimerbyEvent (delegate Custom|TimerExpired) (Variables|ReadOnly|GetQuestTime)))
              (Variables|Default|SetQuestTimer _returnvalue)
              (CallFunction|ShowUI)
              ;[1 1 1] Listen for exit vehicle
              (Vehicle|BindEventtoOnExitVehicle (Fortnite|GetFortPlayerControllerfromActor (Variables|GAVars|Data|GetOwningPlayerPawn)) (delegate Custom|ExitVehicle)))
            (:Added
              ;[1 1 1] Quest complete
              (Quest|SendComplexCustomStatEvent (Variables|GAVars|Data|GetOwningPlayerQuestManager) 0 "(GameplayTags=)" (Variables|ReadOnly|GetQuestStatEventTags) 1)
              (CallFunction|VictoryFX)
              (CallFunction|TimeUp)))
          (else
            (Ability|EndAbility)))))))

(event Custom|ExitVehicle
  ;[1 1 1] Listen for exit vehicle
  (Variables|Default|Setreentered false)
  (Vehicle|BindEventtoOnEnterVehicleDriver (Fortnite|GetFortPlayerControllerfromActor (Variables|GAVars|Data|GetOwningPlayerPawn)) (delegate Custom|Re-Enter%20Vehicle))
  (Utilities|IsinVehicle (Variables|GAVars|Data|GetOwningPlayerPawn)
    (:false
      (CallFunction|CleanUp))))

(event Custom|TimerExpired
  ;[1 1 1] Set timer
  (CallFunction|TimeUp))

(event Custom|Re-Enter%20Vehicle
  ;[1 1 1] Listen for getting back into the vehicle, interrupt cleanup if so
  (bind _vehicle_occupied (Utilities|IsinVehicle (Variables|GAVars|Data|GetOwningPlayerPawn))
    (:true
      (CallFunction|Re-EnteredVehicle)
      (if (GameplayTags|Equal(GameplayTagContainer) (Class|FortVehicleInterface|GetGameplayTags _vehicle_occupied) (Class|FortVehicleInterface|GetGameplayTags (Variables|Default|GetQuestVehicle)))
        (Variables|Default|Setreentered true)))))

(event Custom|Show%20UI)

(event Custom|ExitedVehicle)

(event Custom|Hide%20UI)

(event Custom|Re-EnteredVehicle)

(event Custom|Clean%20Up
  (bind _owningplayerpawn (Variables|GAVars|Data|GetOwningPlayerPawn))
  (bind _returnvalue (Fortnite|GetFortPlayerControllerfromActor _owningplayerpawn))
  ;[1 1 1] Clean up
  (CallFunction|ExitedVehicle)
  (Utilities|FlowControl|Delay 3e+01)
  (if (not (Variables|Default|Getreentered))
    (Utilities|Time|ClearandInvalidateTimerbyHandle (Variables|Default|GetQuestTimer))
    (GameplayTasks|EndTask (Variables|Default|GetTagListenTask))
    (Vehicle|UnbindEventfromOnEnterVehicleDriver _returnvalue (delegate Custom|Re-Enter%20Vehicle))
    (Vehicle|UnbindEventfromOnExitVehicle _returnvalue (delegate Custom|ExitVehicle))
    (CallFunction|HideUI)
    (Variables|Default|Setreentered false)
    (Ability|EndAbility)))

(event Custom|TimeUp
  (bind _owningplayerpawn (Variables|GAVars|Data|GetOwningPlayerPawn))
  (bind _returnvalue (Fortnite|GetFortPlayerControllerfromActor _owningplayerpawn))
  ;[1 1 1] Clean up
  (Utilities|Time|ClearandInvalidateTimerbyHandle (Variables|Default|GetQuestTimer))
  (GameplayTasks|EndTask (Variables|Default|GetTagListenTask))
  (Vehicle|UnbindEventfromOnEnterVehicleDriver _returnvalue (delegate Custom|Re-Enter%20Vehicle))
  (Vehicle|UnbindEventfromOnExitVehicle _returnvalue (delegate Custom|ExitVehicle))
  (CallFunction|HideUI)
  (Variables|Default|Setreentered false)
  (Ability|EndAbility))

(event Custom|VictoryFX
  ;[1 1 1] Victory FX
  (Audio|PlaySound2D (Variables|ReadOnly|GetVictorySound))
  (Effects|SpawnFXSystemAttached (Variables|ReadOnly|GetQuestSuccessEmitter) (Variables|/Script/Engine.Character|GetMesh (Variables|GAVars|Data|GetOwningPlayerPawn)) "Feet"))
